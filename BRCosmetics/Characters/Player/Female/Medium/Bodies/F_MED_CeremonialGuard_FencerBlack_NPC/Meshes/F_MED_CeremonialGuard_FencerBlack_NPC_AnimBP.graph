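# Post-process style cosmetic AnimBP for the female Ceremonial Guard Fencer Black NPC body, adding secondary motion to a pose copied from the attached parent mesh.
# A crouch-aware Control Rig feeds two RigidBody coat simulations with different physics assets, blended by bIsCrouching.
# Face accessory RigidBody follows, then spring-limited AnimDynamics on three sword chain bones.
# A constraint and LookAt finally aim the sword belt pin; the EventGraph is empty.
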AnimBlueprint /BRCosmetics/Characters/Player/Female/Medium/Bodies/F_MED_CeremonialGuard_FencerBlack_NPC/Meshes/F_MED_CeremonialGuard_FencerBlack_NPC_AnimBP
target_skeleton=&/F_MED_CeremonialGuard_FencerBlack_NPC_Skeleton
graphs [
 AnimGraph {nodes [
   Output Pose(Result=Use cached pose 'Post Dynamics'()) {layerGroup=None}
   n1 = Control Rig!(Source=Copy Pose From Mesh() {bUseAttachedParent=true}, Alpha, Crouch:bool=Get bIsCrouching()) -> (Pose) {
    alphaBoolBlend {blendOption=Linear}
    controlRigAssetReference {blueprintRigClass=/Game/Characters/Player/Female/Medium/Bodies/F_MED_CeremonialGuard_FencerBlack_NPC/Meshes/F_MED_CeremonialGuard_FencerBlack_NPC_CtrlRig.F_MED_CeremonialGuard_FencerBlack_NPC_CtrlRig_C}
   }
   Save cached pose 'Pre Dynamics'(Pose=@n1) {cacheName=Pre Dynamics}
   n2 = RigidBody!(ComponentPose=Local To Component(LocalPose=Use cached pose 'Pre Dynamics'()), Alpha, OverrideWorldGravity:Vector="0,0,-980", ExternalForce) -> (Pose) {
    bOverrideWorldGravity=true
    overridePhysicsAsset=/BRCosmetics/Characters/Player/Female/Medium/Bodies/F_MED_CeremonialGuard_Fencer/Meshes/F_MED_CeremonialGuard_Fencer_Physics_Coat_standart
    simulationSpace=BaseBoneSpace
    alphaBoolBlend {blendOption=Linear}
    baseBoneRef {boneName=pelvis}
    componentAppliedLinearAccClamp=<3000 3000 3000>
    componentLinearAccScale {z=1}
    componentLinearVelScale {z=1}
    overrideWorldGravity {z=-400}
    simSpaceSettings {
     maxAngularAcceleration=100
     maxAngularVelocity=3
     maxLinearAcceleration=600
     maxLinearVelocity=150
     velocityScaleZ=0
     worldAlpha=1
     externalLinearDragV=<0.1 0.1>
    }
   }
   n3 = Component To Local(ComponentPose=@n2)
   n4 = RigidBody(ComponentPose=Local To Component(LocalPose=Use cached pose 'Pre Dynamics'()), OverrideWorldGravity="0,0,-980") {
    bOverrideWorldGravity=true
    overridePhysicsAsset=/BRCosmetics/Characters/Player/Female/Medium/Bodies/F_MED_CeremonialGuard_Fencer/Meshes/F_MED_CeremonialGuard_Fencer_Physics_Coat
    simulationSpace=BaseBoneSpace
    alphaBoolBlend {blendOption=Linear}
    baseBoneRef {boneName=pelvis}
    componentAppliedLinearAccClamp=<3000 3000 3000>
    componentLinearAccScale {z=1}
    componentLinearVelScale {z=1}
    overrideWorldGravity {z=-400}
    simSpaceSettings {
     maxAngularAcceleration=100
     maxAngularVelocity=3
     maxLinearAcceleration=600
     maxLinearVelocity=150
     velocityScaleZ=0
     worldAlpha=1
     externalLinearDragV=<0.1 0.1>
    }
   }
   n5 = Component To Local(ComponentPose=@n4)
   n6 = Get bIsCrouching()
   n7 = RigidBody(
    ComponentPose=Local To Component(LocalPose=Blend Poses by bool(BlendPose_0=@n5, BlendPose_1=@n3, bActiveValue=@n6) {blendType=Linear})
    OverrideWorldGravity="0,0,-980"
   ) {
    bOverrideWorldGravity=true
    overridePhysicsAsset=/BRCosmetics/Characters/Player/Female/Medium/Bodies/F_MED_CeremonialGuard_Fencer/Meshes/Parts/F_MED_CeremonialGuard_Fencer_FaceAcc_Physics
    simulationSpace=BaseBoneSpace
    alphaBoolBlend {blendOption=Linear}
    baseBoneRef {boneName=spine_05}
    componentAppliedLinearAccClamp=<7000 7000 7000>
    componentLinearAccScale {z=1}
    componentLinearVelScale {z=1}
    overrideWorldGravity {z=-400}
    simSpaceSettings {
     maxAngularAcceleration=200
     maxAngularVelocity=3
     maxLinearAcceleration=200
     maxLinearVelocity=150
     velocityScaleZ=0
     worldAlpha=1
     externalLinearDragV=<0.1 0.1>
    }
   }
   n8 = Anim Dynamics - Bone: dyn_sword_chain_01_jnt(ComponentPose=@n7, Alpha=1, ExternalForce="0,0,0") {
    angularDampingOverride=1
    angularSpringConstant=1500
    bAngularSpring=true
    bOverrideAngularDamping=true
    bUseGravityOverride=true
    gravityScale=0.35
    simulationSpace=BoneRelative
    alphaBoolBlend {blendOption=Linear}
    boundBone {boneName=dyn_sword_chain_01_jnt}
    componentLinearAccScale=<1 1 1>
    componentLinearVelScale=<1 1 1>
    gravityOverride {z=-980}
    physicsBodyDefinitions [{
      boundBone {boneName=dyn_sword_chain_01_jnt}
      boxExtents=<20 20 20>
      constraintSetup {angularLimitsMax=<0 0 0>, angularLimitsMin=<0 -40 0>, angularTarget=<1 0 0>}
      localJointOffset=<1 0 0>
     }]
    relativeSpaceBone {boneName=dyn_belt_jnt}
   }
   n9 = Anim Dynamics - Bone: dyn_sword_chain_02_jnt(ComponentPose=@n8, Alpha=1, ExternalForce="0,0,0") {
    angularDampingOverride=1
    angularSpringConstant=1500
    bAngularSpring=true
    bOverrideAngularDamping=true
    bUseGravityOverride=true
    gravityScale=0.35
    simulationSpace=BoneRelative
    alphaBoolBlend {blendOption=Linear}
    boundBone {boneName=dyn_sword_chain_02_jnt}
    componentLinearAccScale=<1 1 1>
    componentLinearVelScale=<1 1 1>
    gravityOverride {z=-980}
    physicsBodyDefinitions [{
      boundBone {boneName=dyn_sword_chain_02_jnt}
      boxExtents=<17.5 17.5 17.5>
      constraintSetup {angularLimitsMax=<0 0 0>, angularLimitsMin=<0 -50 0>, angularTarget=<1 0 0>}
      localJointOffset=<1 0 0>
     }]
    relativeSpaceBone {boneName=dyn_belt_jnt}
   }
   n10 = Anim Dynamics - Bone: dyn_sword_chain_03_jnt(ComponentPose=@n9, Alpha=1, ExternalForce="0,0,0") {
    angularDampingOverride=1
    angularSpringConstant=1500
    bAngularSpring=true
    bOverrideAngularDamping=true
    bUseGravityOverride=true
    gravityScale=0.35
    simulationSpace=BoneRelative
    alphaBoolBlend {blendOption=Linear}
    boundBone {boneName=dyn_sword_chain_03_jnt}
    componentLinearAccScale=<1 1 1>
    componentLinearVelScale=<1 1 1>
    gravityOverride {z=-980}
    physicsBodyDefinitions [{
      boundBone {boneName=dyn_sword_chain_03_jnt}
      boxExtents=<15 15 15>
      constraintSetup {angularLimitsMax=<0 0 0>, angularLimitsMin=<0 -60 0>, angularTarget=<1 0 0>}
      localJointOffset=<1 0 0>
     }]
    relativeSpaceBone {boneName=dyn_belt_jnt}
   }
   n11 = "Constraint - dyn_sword_pin_aim_jnt (P)"(ComponentPose=@n10, Alpha=1, ConstraintWeights_0=1) {
    alphaBoolBlend {blendOption=Linear}
    boneToModify {boneName=dyn_sword_pin_aim_jnt}
    constraintSetup [{targetBone {boneName=dyn_sword_pin_end_jnt}}]
    constraintWeights [1]
   }
   n12 = Look At - Bone: dyn_sword_pin_belt_jnt(ComponentPose=@n11, Alpha=1) {
    alphaBoolBlend {blendOption=Linear}
    boneToModify {boneName=dyn_sword_pin_belt_jnt}
    lookAtLocation {x=0}
    lookAtTarget {boneReference {boneName=dyn_sword_pin_aim_jnt}}
    lookAt_Axis {axis=<1 0>}
   }
   Save cached pose 'Post Dynamics'(Pose=Component To Local(ComponentPose=@n12)) {cacheName=Post Dynamics}
  ]}
 EventGraph {links [], nodes []}
]
variables [
 Gravity Override {type=Vector}
 Gravity Override Params {type=GravityOverrideParamsStruct}
]
